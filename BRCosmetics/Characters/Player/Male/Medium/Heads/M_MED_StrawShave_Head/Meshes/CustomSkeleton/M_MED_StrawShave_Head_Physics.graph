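# Physics asset for the M_MED_StrawShave_Head cosmetic, setting up seven three-segment dynamic hair chains alongside head, neck, upper spine, clavicle and upper-arm capsules.
# Each chain hangs from a base bone on the head, with the two side bases attached to the left and right bases.
# Chain joints carry per-chain swing limits and tight twist limits.
# Free-rotation cross-links tie neighboring center, left, right and side chains together, linking them laterally.
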PhysicsAsset /BRCosmetics/Characters/Player/Male/Medium/Heads/M_MED_StrawShave_Head/Meshes/CustomSkeleton/M_MED_StrawShave_Head_Physics
bodies {
 base_hair1 [{shapeType=Capsule, length=0.505, radius=0.505, shapeName=base_hair1_capsule, rotation=<rot 0 -90 0>}]
 base_hair7 [{shapeType=Capsule, length=0.505, radius=0.505, shapeName=base_hair7_capsule, rotation=<rot 0 -90 0>}]
 base_hair_c [{shapeType=Capsule, length=0.505, radius=0.505, shapeName=base_hair_c_capsule, rotation=<rot 0 -90 0>}]
 base_hair_l [{shapeType=Capsule, length=0.505, radius=0.505, shapeName=base_hair_l_capsule, rotation=<rot 0 -90 0>}]
 base_hair_r [{shapeType=Capsule, length=1, radius=1, shapeName=base_hair_r_capsule, center=<1 0 0>, rotation=<rot -90 0 0>}]
 base_hair_side_l [{shapeType=Capsule, length=0.505, radius=0.505, shapeName=base_hair_side_l_capsule, rotation=<rot 0 -90 0>}]
 base_hair_side_r [{shapeType=Capsule, length=0.505, radius=0.505, shapeName=base_hair_side_r_capsule, rotation=<rot 0 -90 0>}]
 clavicle_l [{shapeType=Capsule, length=15, radius=10, shapeName=clavicle_r_capsule, center=<10 1 0>, rotation=<rot -90 0 0>}]
 clavicle_r [{shapeType=Capsule, length=15, radius=10, shapeName=clavicle_r_capsule, center=<-10 1 0>, rotation=<rot -90 0 0>}]
 dyn_hair1_1 [{shapeType=Capsule, length=5, radius=4, shapeName=dyn_hair1_1_capsule, center=<2 1 1.1>, rotation=<rot 61.8105 90 -180>}]
 dyn_hair1_2 [{shapeType=Capsule, length=5, radius=4, shapeName=dyn_hair1_2_capsule, center=<2.2734 0.9 2.8>, rotation=<rot 63.8136 90 -180>}]
 dyn_hair1_3 [{shapeType=Capsule, length=5, radius=4, shapeName=dyn_hair1_3_capsule, center=<4.7566 1 3.1>, rotation=<rot 63.8136 90.0001 169.0097>}]
 dyn_hair7_1 [{shapeType=Capsule, length=5, radius=4, shapeName=dyn_hair7_1_capsule, center=<2 0.3 1.4>, rotation=<rot 58.8781 -90.0085 -0.0012>}]
 dyn_hair7_2 [{shapeType=Capsule, length=5, radius=4, shapeName=dyn_hair7_2_capsule, center=<3.1252 -0.5 2>, rotation=<rot 65.4985 -90 0>}]
 dyn_hair7_3 [{shapeType=Capsule, length=5, radius=4, shapeName=dyn_hair7_3_capsule, center=<6.1683 -1.9 2.7>, rotation=<rot 62.8839 -90.0001 -3.0963>}]
 dyn_hair_c_01 [{shapeType=Capsule, length=6, radius=5, center=<6 0 7>, rotation=<rot 0 0 90>}]
 dyn_hair_c_02 [{shapeType=Capsule, length=6, radius=5, center=<6 0 7>, rotation=<rot 0 0 90>}]
 dyn_hair_c_03 [{shapeType=Capsule, length=6, radius=6, center=<6 0 7>, rotation=<rot 0 0 90>}]
 dyn_hair_l_01 [{shapeType=Capsule, length=6, radius=5, center=<6 0 7>, rotation=<rot 0 0 90>}]
 dyn_hair_l_02 [{shapeType=Capsule, length=6, radius=5, center=<6 0 7>, rotation=<rot 0 0 90>}]
 dyn_hair_l_03 [{shapeType=Capsule, length=6, radius=6, center=<6 0 7>, rotation=<rot 0 0 90>}]
 dyn_hair_r_01 [{shapeType=Capsule, length=6, radius=5, center=<6 0 7>, rotation=<rot 0 0 90>}]
 dyn_hair_r_02 [{shapeType=Capsule, length=6, radius=5, center=<6 0 7>, rotation=<rot 0 0 90>}]
 dyn_hair_r_03 [{shapeType=Capsule, length=6, radius=6, center=<6 0 7>, rotation=<rot 0 0 90>}]
 dyn_hair_side_l_01 [{shapeType=Capsule, length=10, radius=5, center=<4 0 6>, rotation=<rot 0 0 90>}]
 dyn_hair_side_l_02 [{shapeType=Capsule, length=10, radius=5, center=<4 0 6>, rotation=<rot 0 0 90>}]
 dyn_hair_side_l_03 [{shapeType=Capsule, length=10, radius=5, center=<4 0 6>, rotation=<rot 0 0 90>}]
 dyn_hair_side_r_01 [{shapeType=Capsule, length=10, radius=5, center=<4 0 6>, rotation=<rot 0 0 90>}]
 dyn_hair_side_r_02 [{shapeType=Capsule, length=10, radius=5, center=<4 0 6>, rotation=<rot 0 0 90>}]
 dyn_hair_side_r_03 [{shapeType=Capsule, length=10, radius=6, center=<4 0 4>, rotation=<rot 0 0 90>}]
 head [{shapeType=Capsule, length=15, radius=14, shapeName=head_capsule, center=<9.3533 0 0>, rotation=<rot 90 0 0>}]
 neck_01 [{shapeType=Capsule, length=10, radius=9, shapeName=head_capsule, center=<4 0 0>, rotation=<rot 90 0 0>}]
 neck_02 [{shapeType=Capsule, length=10, radius=12, shapeName=head_capsule, center=<4 0 0>, rotation=<rot 90 0 0>}]
 spine_03 [{shapeType=Capsule, length=15, radius=20}]
 spine_04 [{shapeType=Capsule, length=15, radius=20}]
 spine_05 [{shapeType=Capsule, length=15, radius=15}]
 upperarm_l [{shapeType=Capsule, length=20, radius=11, shapeName=upperarm_l_capsule, center=<13 0 0>, rotation=<rot -90 0 0>}]
 upperarm_r [{shapeType=Capsule, length=20, radius=11, shapeName=upperarm_l_capsule, center=<-12.5948 0 -2.2978>, rotation=<rot -80 0.0004 -0.0004>}]
}
constraints [
 {
  bone1Name=base_hair1
  bone2Name=head
  swing1LimitDegrees=45
  swing1Motion=Free
  swing2LimitDegrees=45
  swing2Motion=Free
  twistLimitDegrees=45
  twistMotion=Free
 }
 {bone1Name=dyn_hair1_1, bone2Name=base_hair1, swing1LimitDegrees=15, swing2LimitDegrees=15, twistLimitDegrees=1}
 {bone1Name=dyn_hair1_2, bone2Name=dyn_hair1_1, swing1LimitDegrees=15, swing2LimitDegrees=15, twistLimitDegrees=1}
 {bone1Name=dyn_hair1_3, bone2Name=dyn_hair1_2, swing1LimitDegrees=15, swing2LimitDegrees=15, twistLimitDegrees=1}
 {
  bone1Name=base_hair7
  bone2Name=head
  swing1LimitDegrees=45
  swing1Motion=Free
  swing2LimitDegrees=45
  swing2Motion=Free
  twistLimitDegrees=45
  twistMotion=Free
 }
 {bone1Name=dyn_hair7_1, bone2Name=base_hair7, swing1LimitDegrees=15, swing2LimitDegrees=15, twistLimitDegrees=1}
 {bone1Name=dyn_hair7_2, bone2Name=dyn_hair7_1, swing1LimitDegrees=15, swing2LimitDegrees=15, twistLimitDegrees=1}
 {bone1Name=dyn_hair7_3, bone2Name=dyn_hair7_2, swing1LimitDegrees=15, swing2LimitDegrees=15, twistLimitDegrees=1}
 {bone1Name=dyn_hair_r_03, bone2Name=dyn_hair_r_02, swing1LimitDegrees=15, swing2LimitDegrees=25, twistLimitDegrees=2}
 {
  bone1Name=base_hair_c
  bone2Name=head
  swing1LimitDegrees=45
  swing1Motion=Free
  swing2LimitDegrees=45
  swing2Motion=Free
  twistLimitDegrees=45
  twistMotion=Free
 }
 {bone1Name=dyn_hair_c_01, bone2Name=base_hair_c, swing1LimitDegrees=10, swing2LimitDegrees=25, twistLimitDegrees=2}
 {bone1Name=dyn_hair_c_02, bone2Name=dyn_hair_c_01, swing1LimitDegrees=10, swing2LimitDegrees=25, twistLimitDegrees=2}
 {bone1Name=dyn_hair_c_03, bone2Name=dyn_hair_c_02, swing1LimitDegrees=10, swing2LimitDegrees=25, twistLimitDegrees=2}
 {
  bone1Name=base_hair_l
  bone2Name=head
  swing1LimitDegrees=45
  swing1Motion=Free
  swing2LimitDegrees=45
  swing2Motion=Free
  twistLimitDegrees=45
  twistMotion=Free
 }
 {bone1Name=dyn_hair_l_01, bone2Name=base_hair_l, swing1LimitDegrees=15, swing2LimitDegrees=25, twistLimitDegrees=2}
 {bone1Name=dyn_hair_l_02, bone2Name=dyn_hair_l_01, swing1LimitDegrees=15, swing2LimitDegrees=25, twistLimitDegrees=2}
 {bone1Name=dyn_hair_l_03, bone2Name=dyn_hair_l_02, swing1LimitDegrees=15, swing2LimitDegrees=25, twistLimitDegrees=2}
 {
  bone1Name=base_hair_r
  bone2Name=head
  swing1LimitDegrees=45
  swing1Motion=Free
  swing2LimitDegrees=45
  swing2Motion=Free
  twistLimitDegrees=45
  twistMotion=Free
 }
 {bone1Name=dyn_hair_r_01, bone2Name=base_hair_r, swing1LimitDegrees=15, swing2LimitDegrees=25, twistLimitDegrees=2}
 {bone1Name=dyn_hair_r_02, bone2Name=dyn_hair_r_01, swing1LimitDegrees=15, swing2LimitDegrees=25, twistLimitDegrees=2}
 {
  bone1Name=dyn_hair_r_02
  bone2Name=dyn_hair_c_02
  swing1LimitDegrees=45
  swing1Motion=Free
  swing2LimitDegrees=45
  swing2Motion=Free
  twistLimitDegrees=45
  twistMotion=Free
 }
 {
  bone1Name=dyn_hair_l_02
  bone2Name=dyn_hair_c_02
  swing1LimitDegrees=45
  swing1Motion=Free
  swing2LimitDegrees=45
  swing2Motion=Free
  twistLimitDegrees=45
  twistMotion=Free
 }
 {bone1Name=base_hair_side_l, bone2Name=base_hair_l, swing1LimitDegrees=15, swing2LimitDegrees=10, twistLimitDegrees=2}
 {bone1Name=base_hair_side_r, bone2Name=base_hair_r, swing1LimitDegrees=15, swing2LimitDegrees=15, twistLimitDegrees=2}
 {
  bone1Name=dyn_hair_c_03
  bone2Name=dyn_hair_l_03
  swing1LimitDegrees=45
  swing1Motion=Free
  swing2LimitDegrees=45
  swing2Motion=Free
  twistLimitDegrees=45
  twistMotion=Free
 }
 {
  bone1Name=dyn_hair_c_03
  bone2Name=dyn_hair_r_03
  swing1LimitDegrees=45
  swing1Motion=Free
  swing2LimitDegrees=45
  swing2Motion=Free
  twistLimitDegrees=45
  twistMotion=Free
 }
 {
  bone1Name=neck_01
  bone2Name=spine_05
  swing1LimitDegrees=45
  swing1Motion=Free
  swing2LimitDegrees=45
  swing2Motion=Free
  twistLimitDegrees=45
  twistMotion=Free
 }
 {
  bone1Name=neck_02
  bone2Name=spine_05
  swing1LimitDegrees=45
  swing1Motion=Free
  swing2LimitDegrees=45
  swing2Motion=Free
  twistLimitDegrees=45
  twistMotion=Free
 }
 {
  bone1Name=head
  bone2Name=spine_05
  swing1LimitDegrees=45
  swing1Motion=Free
  swing2LimitDegrees=45
  swing2Motion=Free
  twistLimitDegrees=45
  twistMotion=Free
 }
 {
  bone1Name=clavicle_l
  bone2Name=spine_05
  swing1LimitDegrees=45
  swing1Motion=Free
  swing2LimitDegrees=45
  swing2Motion=Free
  twistLimitDegrees=45
  twistMotion=Free
 }
 {
  bone1Name=clavicle_r
  bone2Name=spine_05
  swing1LimitDegrees=45
  swing1Motion=Free
  swing2LimitDegrees=45
  swing2Motion=Free
  twistLimitDegrees=45
  twistMotion=Free
 }
 {
  bone1Name=spine_05
  bone2Name=spine_04
  swing1LimitDegrees=45
  swing1Motion=Free
  swing2LimitDegrees=45
  swing2Motion=Free
  twistLimitDegrees=45
  twistMotion=Free
 }
 {
  bone1Name=spine_04
  bone2Name=spine_03
  swing1LimitDegrees=45
  swing1Motion=Free
  swing2LimitDegrees=45
  swing2Motion=Free
  twistLimitDegrees=45
  twistMotion=Free
 }
 {
  bone1Name=upperarm_r
  bone2Name=clavicle_r
  swing1LimitDegrees=45
  swing1Motion=Free
  swing2LimitDegrees=45
  swing2Motion=Free
  twistLimitDegrees=45
  twistMotion=Free
 }
 {
  bone1Name=upperarm_l
  bone2Name=clavicle_l
  swing1LimitDegrees=45
  swing1Motion=Free
  swing2LimitDegrees=45
  swing2Motion=Free
  twistLimitDegrees=45
  twistMotion=Free
 }
 {bone1Name=dyn_hair_side_r_01, bone2Name=base_hair_side_r, swing1LimitDegrees=15, swing2LimitDegrees=45, twistLimitDegrees=2}
 {bone1Name=dyn_hair_side_r_02, bone2Name=dyn_hair_side_r_01, swing1LimitDegrees=15, swing2LimitDegrees=45, twistLimitDegrees=2}
 {bone1Name=dyn_hair_side_r_03, bone2Name=dyn_hair_side_r_02, swing1LimitDegrees=15, swing2LimitDegrees=45, twistLimitDegrees=2}
 {bone1Name=dyn_hair_side_l_01, bone2Name=base_hair_side_l, swing1LimitDegrees=45, swing2LimitDegrees=45, twistLimitDegrees=2}
 {bone1Name=dyn_hair_side_l_02, bone2Name=dyn_hair_side_l_01, swing1LimitDegrees=45, swing2LimitDegrees=45, twistLimitDegrees=2}
 {bone1Name=dyn_hair_side_l_03, bone2Name=dyn_hair_side_l_02, swing1LimitDegrees=45, swing2LimitDegrees=45, twistLimitDegrees=2}
 {
  bone1Name=dyn_hair_side_r_03
  bone2Name=dyn_hair_r_03
  swing1LimitDegrees=45
  swing1Motion=Free
  swing2LimitDegrees=45
  swing2Motion=Free
  twistLimitDegrees=45
  twistMotion=Free
 }
 {
  bone1Name=dyn_hair_side_l_03
  bone2Name=dyn_hair_l_03
  swing1LimitDegrees=45
  swing1Motion=Free
  swing2LimitDegrees=45
  swing2Motion=Free
  twistLimitDegrees=45
  twistMotion=Free
 }
]
