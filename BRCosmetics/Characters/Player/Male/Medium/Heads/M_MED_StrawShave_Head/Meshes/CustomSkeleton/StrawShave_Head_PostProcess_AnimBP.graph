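# Post-process AnimBP for the StrawShave head that drives hair bone motion.
# A Control Rig pose feeds two head-space RigidBody sims, switched by skydiving, then half-strength constraints steady the side hair.
# Backblings, capes or disabled physics switch output to a simulation-free branch constraining hair to head, neck, spine and clavicles.
# Each update detects backblings and computes movement-state gravity and RigidBody settings via Fortnite's gravity-override function.
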AnimBlueprint /BRCosmetics/Characters/Player/Male/Medium/Heads/M_MED_StrawShave_Head/Meshes/CustomSkeleton/StrawShave_Head_PostProcess_AnimBP
target_skeleton=&/Fortnite_Base_Head_StrawShave_Skeleton_New
graphs [
 AnimGraph {nodes [
   n1 = "Transform (Modify) Bone - Bone: base_hair_side_r"!(ComponentPose, bAlphaBoolEnabled:bool=True, Translation="-3,0,0", Rotation="0,0,0", Scale="0,1,1") -> (Pose) {
    alphaInputType=Bool
    rotationMode=BMM_Additive
    rotationSpace=BCS_BoneSpace
    translationMode=BMM_Additive
    translationSpace=BCS_BoneSpace
    alphaBoolBlend {blendInTime=0.2, blendOutTime=0.2}
    boneToModify {boneName=base_hair_side_r}
    scale {x=0}
    translation {x=-3}
   }
   "Transform (Modify) Bone - Bone: base_hair_side_l"!(ComponentPose=@n1, bAlphaBoolEnabled:bool=True, Translation="-3,0,0", Rotation="0,0,0", Scale="1,1,1") -> (Pose) {
    alphaInputType=Bool
    rotationMode=BMM_Additive
    rotationSpace=BCS_BoneSpace
    translationMode=BMM_Additive
    translationSpace=BCS_BoneSpace
    alphaBoolBlend {blendInTime=0.2, blendOutTime=0.2}
    boneToModify {boneName=base_hair_side_l}
    translation {x=-3}
   }
   n2 = "Constraint - dyn_hair_side_l_01 (PP)"!(ComponentPose=Local To Component(LocalPose=Use cached pose 'RB_cache'()), Alpha=0.5, ConstraintWeights_0=1, ConstraintWeights_1:float=0) -> (Pose) {
    alphaBoolBlend {blendInTime=0.2, blendOption=Linear, blendOutTime=0.2}
    boneToModify {boneName=dyn_hair_side_l_01}
    constraintSetup [{targetBone {boneName=head}}, {targetBone {boneName=neck_02}}]
    constraintWeights [1,1]
   }
   n3 = "Constraint - dyn_hair_side_r_01 (PP)"!(ComponentPose=@n2, Alpha=0.5, ConstraintWeights_0=1, ConstraintWeights_1:float=0) -> (Pose) {
    alphaBoolBlend {blendInTime=0.2, blendOption=Linear, blendOutTime=0.2}
    boneToModify {boneName=dyn_hair_side_r_01}
    constraintSetup [{targetBone {boneName=head}}, {targetBone {boneName=neck_02}}]
    constraintWeights [1,1]
   }
   n4 = "Constraint - dyn_hair_side_l_02 (PP)"!(ComponentPose=@n3, Alpha=0.5, ConstraintWeights_0=1, ConstraintWeights_1:float=0) -> (Pose) {
    alphaBoolBlend {blendInTime=0.2, blendOption=Linear, blendOutTime=0.2}
    boneToModify {boneName=dyn_hair_side_l_02}
    constraintSetup [{targetBone {boneName=head}}, {targetBone {boneName=neck_02}}]
    constraintWeights [1,1]
   }
   n5 = "Constraint - dyn_hair_side_r_02 (PP)"!(ComponentPose=@n4, Alpha=0.5, ConstraintWeights_0=1, ConstraintWeights_1:float=0) -> (Pose) {
    alphaBoolBlend {blendInTime=0.2, blendOption=Linear, blendOutTime=0.2}
    boneToModify {boneName=dyn_hair_side_r_02}
    constraintSetup [{targetBone {boneName=head}}, {targetBone {boneName=neck_02}}]
    constraintWeights [1,1]
   }
   n6 = "Constraint - dyn_hair_side_l_03 (PPP)"!(ComponentPose=@n5, Alpha=0.5, ConstraintWeights_0=1, ConstraintWeights_1:float=0, ConstraintWeights_2:float=0) -> (Pose) {
    alphaBoolBlend {blendInTime=0.2, blendOption=Linear, blendOutTime=0.2}
    boneToModify {boneName=dyn_hair_side_l_03}
    constraintSetup [{targetBone {boneName=neck_01}}, {targetBone {boneName=spine_05}}, {targetBone {boneName=clavicle_l}}]
    constraintWeights [1,1,1]
   }
   n7 = "Constraint - dyn_hair_side_r_03 (PPP)"!(ComponentPose=@n6, Alpha=0.5, ConstraintWeights_0=1, ConstraintWeights_1:float=0, ConstraintWeights_2:float=0) -> (Pose) {
    alphaBoolBlend {blendInTime=0.2, blendOption=Linear, blendOutTime=0.2}
    boneToModify {boneName=dyn_hair_side_r_03}
    constraintSetup [{targetBone {boneName=neck_01}}, {targetBone {boneName=spine_05}}, {targetBone {boneName=clavicle_r}}]
    constraintWeights [1,1,1]
   }
   n8 = "Constraint - base_hair14 (PPP)"!(ComponentPose=@n7, Alpha=0.8, ConstraintWeights_0=1, ConstraintWeights_1:float=0, ConstraintWeights_2:float=0) -> (Pose) {
    alphaBoolBlend {blendOption=Linear}
    alphaScaleBiasClamp {bClampResult=true, scale=0.3}
    boneToModify {boneName=base_hair14}
    constraintSetup [{targetBone {boneName=neck_01}}, {targetBone {boneName=spine_05}}, {targetBone {boneName=clavicle_r}}]
    constraintWeights [1,1,1]
   }
   n9 = "Constraint - dyn_hair_r_01 (PP)"!(ComponentPose=Local To Component(LocalPose=Use cached pose 'CtrlRig_cache'()), Alpha=1, ConstraintWeights_0=1, ConstraintWeights_1:float=0) -> (Pose) {
    alphaBoolBlend {blendOption=Linear}
    boneToModify {boneName=dyn_hair_r_01}
    constraintSetup [{targetBone {boneName=head}}, {targetBone {boneName=neck_02}}]
    constraintWeights [1,1]
   }
   n10 = "Constraint - dyn_hair_r_02 (PP)"!(ComponentPose=@n9, Alpha=1, ConstraintWeights_0=1, ConstraintWeights_1:float=0) -> (Pose) {
    alphaBoolBlend {blendOption=Linear}
    boneToModify {boneName=dyn_hair_r_02}
    constraintSetup [{targetBone {boneName=head}}, {targetBone {boneName=neck_02}}]
    constraintWeights [1,1]
   }
   n11 = "Constraint - dyn_hair_r_03 (PPP)"!(ComponentPose=@n10, Alpha=1, ConstraintWeights_0=1, ConstraintWeights_1:float=0, ConstraintWeights_2:float=0) -> (Pose) {
    alphaBoolBlend {blendOption=Linear}
    boneToModify {boneName=dyn_hair_r_03}
    constraintSetup [{targetBone {boneName=neck_01}}, {targetBone {boneName=spine_05}}, {targetBone {boneName=clavicle_r}}]
    constraintWeights [1,1,1]
   }
   n12 = "Constraint - dyn_hair_l_01 (PP)"!(ComponentPose=@n11, Alpha=1, ConstraintWeights_0=1, ConstraintWeights_1:float=0) -> (Pose) {
    alphaBoolBlend {blendOption=Linear}
    boneToModify {boneName=dyn_hair_l_01}
    constraintSetup [{targetBone {boneName=head}}, {targetBone {boneName=neck_02}}]
    constraintWeights [1,1]
   }
   n13 = "Constraint - dyn_hair_l_02 (PP)"!(ComponentPose=@n12, Alpha=1, ConstraintWeights_0=1, ConstraintWeights_1:float=0) -> (Pose) {
    alphaBoolBlend {blendOption=Linear}
    boneToModify {boneName=dyn_hair_l_02}
    constraintSetup [{targetBone {boneName=head}}, {targetBone {boneName=neck_02}}]
    constraintWeights [1,1]
   }
   n14 = "Constraint - dyn_hair_l_03 (PPP)"!(ComponentPose=@n13, Alpha=1, ConstraintWeights_0=1, ConstraintWeights_1:float=0, ConstraintWeights_2:float=0) -> (Pose) {
    alphaBoolBlend {blendOption=Linear}
    boneToModify {boneName=dyn_hair_l_03}
    constraintSetup [{targetBone {boneName=neck_01}}, {targetBone {boneName=spine_05}}, {targetBone {boneName=clavicle_l}}]
    constraintWeights [1,1,1]
   }
   n15 = "Constraint - dyn_hair_c_01 (PP)"!(ComponentPose=@n14, Alpha=1, ConstraintWeights_0=1, ConstraintWeights_1:float=0) -> (Pose) {
    alphaBoolBlend {blendOption=Linear}
    boneToModify {boneName=dyn_hair_c_01}
    constraintSetup [{targetBone {boneName=head}}, {targetBone {boneName=neck_02}}]
    constraintWeights [1,1]
   }
   n16 = "Constraint - dyn_hair_c_02 (PP)"!(ComponentPose=@n15, Alpha=1, ConstraintWeights_0=0.5, ConstraintWeights_1:float=0.5) -> (Pose) {
    alphaBoolBlend {blendOption=Linear}
    boneToModify {boneName=dyn_hair_c_02}
    constraintSetup [{targetBone {boneName=head}}, {targetBone {boneName=neck_02}}]
    constraintWeights [1,1]
   }
   n17 = "Constraint - dyn_hair_c_03 (PP)"!(ComponentPose=@n16, Alpha=1, ConstraintWeights_0=1, ConstraintWeights_1:float=0) -> (Pose) {
    alphaBoolBlend {blendOption=Linear}
    boneToModify {boneName=dyn_hair_c_03}
    constraintSetup [{targetBone {boneName=neck_01}}, {targetBone {boneName=spine_05}}]
    constraintWeights [1,1]
   }
   n18 = "Constraint - dyn_hair_side_l_01 (PP)"(ComponentPose=@n17, Alpha=1, ConstraintWeights_0=1, ConstraintWeights_1=0) {
    alphaBoolBlend {blendOption=Linear}
    boneToModify {boneName=dyn_hair_side_l_01}
    constraintSetup [{targetBone {boneName=head}}, {targetBone {boneName=neck_02}}]
    constraintWeights [1,1]
   }
   n19 = "Constraint - dyn_hair_side_r_01 (PP)"(ComponentPose=@n18, Alpha=1, ConstraintWeights_0=1, ConstraintWeights_1=0) {
    alphaBoolBlend {blendOption=Linear}
    boneToModify {boneName=dyn_hair_side_r_01}
    constraintSetup [{targetBone {boneName=head}}, {targetBone {boneName=neck_02}}]
    constraintWeights [1,1]
   }
   n20 = "Constraint - dyn_hair_side_l_02 (PP)"(ComponentPose=@n19, Alpha=1, ConstraintWeights_0=1, ConstraintWeights_1=0) {
    alphaBoolBlend {blendOption=Linear}
    boneToModify {boneName=dyn_hair_side_l_02}
    constraintSetup [{targetBone {boneName=head}}, {targetBone {boneName=neck_02}}]
    constraintWeights [1,1]
   }
   n21 = "Constraint - dyn_hair_side_r_02 (PP)"(ComponentPose=@n20, Alpha=1, ConstraintWeights_0=1, ConstraintWeights_1=0) {
    alphaBoolBlend {blendOption=Linear}
    boneToModify {boneName=dyn_hair_side_r_02}
    constraintSetup [{targetBone {boneName=head}}, {targetBone {boneName=neck_02}}]
    constraintWeights [1,1]
   }
   n22 = "Constraint - dyn_hair_side_l_03 (PPP)"(ComponentPose=@n21, Alpha=1, ConstraintWeights_0=1, ConstraintWeights_1=0, ConstraintWeights_2=0) {
    alphaBoolBlend {blendOption=Linear}
    boneToModify {boneName=dyn_hair_side_l_03}
    constraintSetup [{targetBone {boneName=neck_01}}, {targetBone {boneName=spine_05}}, {targetBone {boneName=clavicle_l}}]
    constraintWeights [1,1,1]
   }
   n23 = "Constraint - dyn_hair_side_r_03 (PPP)"(ComponentPose=@n22, Alpha=1, ConstraintWeights_0=1, ConstraintWeights_1=0, ConstraintWeights_2=0) {
    alphaBoolBlend {blendOption=Linear}
    boneToModify {boneName=dyn_hair_side_r_03}
    constraintSetup [{targetBone {boneName=neck_01}}, {targetBone {boneName=spine_05}}, {targetBone {boneName=clavicle_r}}]
    constraintWeights [1,1,1]
   }
   n24 = "Constraint - dyn_hair14_1 (PPP)"!(ComponentPose=@n23, Alpha=1, ConstraintWeights_0=1, ConstraintWeights_1:float=0, ConstraintWeights_2:float=0) -> (Pose) {
    alphaBoolBlend {blendOption=Linear}
    boneToModify {boneName=dyn_hair14_1}
    constraintSetup [{targetBone {boneName=head}}, {targetBone {boneName=spine_05}}, {targetBone {boneName=clavicle_r}}]
    constraintWeights [1,1,1]
   }
   n25 = NOT Boolean(A=Get bRigidBodyEnabled())
   n26 = Get Is Wearing Cape() -> (Is Wearing Cape:bool)
   n27 = Get isWearingBackbling()
   Output Pose(Result=Blend Poses by bool(
     BlendPose_0=Component To Local(ComponentPose=@n24)
     BlendPose_1=Component To Local(ComponentPose=@n8)
     bActiveValue=OR Boolean!(A=@n27, B=@n26, C:bool=@n25) -> (ReturnValue)
    ) {blendType=Linear})
   n28 = Control Rig(Source=Input Pose()) {controlRigAssetReference {controlRigAsset=&/M_MED_StrawShave_Head_CtrRig}}
   Save cached pose 'CtrlRig_cache'(Pose=@n28) {cacheName=CtrlRig_cache}
   n29 = RigidBody!(
    ComponentPose=Local To Component(LocalPose=Use cached pose 'CtrlRig_cache'())
    Alpha
    OverrideWorldGravity:Vector="0,0,0"
    ExternalForce
    SimSpaceSettings:Sim Space Settings Structure=(WorldAlpha=0,VelocityScaleZ=1,DampingAlpha=1,MaxLinearVelocity=10000,MaxAngularVelocity=10000,MaxLinearAcceleration=10000,MaxAngularAcceleration=10000,ExternalLinearDragV=(X=0,Y=0,Z=0),ExternalLinearVelocity=(X=0,Y=0,Z=0),ExternalAngularVelocity=(X=0,Y=0,Z=0))
   ) -> (Pose) {
    bClampLinearTranslationLimitToRefPose=true
    bOverrideWorldGravity=true
    overridePhysicsAsset=&/M_MED_StrawShave_Head_Physics
    simulationSpace=BaseBoneSpace
    baseBoneRef {boneName=head}
    componentAppliedLinearAccClamp=<0 0 0>
   }
   n30 = RigidBody(
    ComponentPose=Local To Component(LocalPose=Use cached pose 'CtrlRig_cache'())
    OverrideWorldGravity="0,0,0"
    SimSpaceSettings=(WorldAlpha=0,VelocityScaleZ=1,DampingAlpha=1,MaxLinearVelocity=10000,MaxAngularVelocity=10000,MaxLinearAcceleration=10000,MaxAngularAcceleration=10000,ExternalLinearDragV=(X=0,Y=0,Z=0),ExternalLinearVelocity=(X=0,Y=0,Z=0),ExternalAngularVelocity=(X=0,Y=0,Z=0))
   ) {
    bClampLinearTranslationLimitToRefPose=true
    bOverrideWorldGravity=true
    overridePhysicsAsset=&/M_MED_StrawShave_Head_sky_Physics
    simulationSpace=BaseBoneSpace
    baseBoneRef {boneName=head}
    componentAppliedLinearAccClamp=<0 0 0>
   }
   Save cached pose 'RB_cache'(Pose=Blend Poses by bool(BlendPose_0=Component To Local(ComponentPose=@n30), BlendPose_1=Component To Local(ComponentPose=@n29), bActiveValue=Get bIsSkyDiving())) {cacheName=RB_cache}
   Get bIsPlayingEmote()
  ]}
 EventGraph {nodes [
   Event BlueprintUpdateAnimation()
   > n1 = Sequence()
   > n2 = "Gravity Override State Function (Nativized)"!(
    execute
    CharacterPartAnimInstance:Custom Character Part Anim Instance Object Reference=Self-Reference()
    GravityOverrideParameterSet:Fort Gravity Override Parameter Set Structure=Get Gravity Override Params() -> (Gravity Override Params:Fort Gravity Override Parameter Set Structure)
    RigidBodyAnimNodeParameterSet:Fort Rigid Body Anim Node Parameter Set Structure=Get RBAN Params() -> (RBAN Params:Fort Rigid Body Anim Node Parameter Set Structure)
    DebugDraw:bool=false
    GameTimeInSeconds:float=GetGameTimeInSeconds()
    DebugJointName:Name=root
    StateLogicLayerName:Name=none
   ) -> (
    then
    OutGravityOverride:Vector
    OutCurrentState:Name=None
    OutGravityOverrideParameters:Fort Gravity Override Parameters Structure
    OutRigidBodyAnimNodeParameters:Fort Rigid Body Anim Node Parameters Structure
   )
   > Set Gravity Override(Gravity Override=@n2)
   > Set RBAN Settings Out(RBAN Settings Out=@n2.OutRigidBodyAnimNodeParameters)
   n3 = CheckForBackbling()
   Set isWearingBackbling(
    execute=@n1.then_1
    isWearingBackbling=AND Boolean!(A=@n3, B=NOT Boolean(A=@n3.bIsBackblingACape), C:bool=NOT Boolean(A=@n3.bIsBackblingLong)) -> (ReturnValue)
   )
   > Set Is Wearing Cape(Is Wearing Cape=OR Boolean!(A=@n3.bIsBackblingACape, B=@n3.bIsBackblingLong) -> (ReturnValue))
   TryGetPawnOwner()
  ]}
]
variables [
 Gravity Override {type=Vector}
 Gravity Override Params {
  type=FortGravityOverrideParameterSet
  default=(Skydive_GoingUp=(GravityOverride=(WindFrequency=(X=1.500000,Y=7.000000,Z=8.000000),WindAmplitude=(X=200.000000,Y=-500.000000,Z=400.000000),GravityOverride=(X=0.000000,Y=-300.000000,Z=100.000000),JointName="root")),Skydive_FloatingDown=(GravityOverride=(WindFrequency=(X=1.500000,Y=7.000000,Z=8.000000),WindAmplitude=(X=200.000000,Y=-500.000000,Z=400.000000),WindOffset=(X=0.000000,Y=-600.000000,Z=400.000000),GravityOverride=(X=0.000000,Y=-700.000000,Z=600.000000),JointName="root")),Skydive_DivingDown=(GravityOverride=(WindFrequency=(X=1.500000,Y=12.000000,Z=10.000000),WindAmplitude=(X=50.000000,Y=300.000000,Z=100.000000),GravityOverride=(X=-1500.000000,Y=0.000000,Z=0.000000),JointName="pelvis")),Skydive_Parachuting=(GravityOverride=(WindFrequency=(X=1.500000,Y=7.000000,Z=8.000000),WindAmplitude=(X=200.000000,Y=-500.000000,Z=400.000000),GravityOverride=(X=0.000000,Y=-800.000000,Z=-980.000000),JointName="root")),OnGround_Standing=(GravityOverride=(GravityOverride=(X=0.000000,Y=0.000000,Z=-980.000000),JointName="root")),OnGround_Moving=(GravityOverride=(GravityOverride=(X=0.000000,Y=0.000000,Z=-980.000000),JointName="root")),OnGround_Moving_HighRotationRate=(GravityOverride=(GravityOverride=(X=0.000000,Y=0.000000,Z=-980.000000),JointName="root")),OnGround_Crouch_Moving=(GravityOverride=(GravityOverride=(X=0.000000,Y=0.000000,Z=-980.000000),JointName="root")),Emote_or_Melee=(GravityOverride=(GravityOverride=(X=0.000000,Y=0.000000,Z=-980.000000),JointName="root")),Swimming=(GravityOverride=(GravityOverride=(X=-800.000000,Y=0.000000,Z=0.000000),JointName="pelvis")),Front_End=(GravityOverride=(GravityOverride=(X=0.000000,Y=0.000000,Z=-980.000000),JointName="root")),DBNO=(GravityOverride=(GravityOverride=(X=-980.000000,Y=400.000000,Z=0.000000),JointName="pelvis")),DBNO_Carried=(GravityOverride=(GravityOverride=(X=-980.000000,Y=500.000000,Z=-500.000000),JointName="pelvis")),OnGround_Falling=(GravityOverride=(GravityOverride=(X=0.000000,Y=0.000000,Z=-980.000000),JointName="root")),Ziplining=(GravityOverride=(GravityOverride=(X=0.000000,Y=0.000000,Z=-980.000000),JointName="root")),Skydive_Surfing=(GravityOverride=(WindFrequency=(X=1.500000,Y=7.000000,Z=8.000000),WindAmplitude=(X=200.000000,Y=-500.000000,Z=400.000000),GravityOverride=(X=500.000000,Y=-300.000000,Z=-980.000000),JointName="root")),Riding_Standing=(GravityOverride=(GravityOverride=(X=0.000000,Y=-350.000000,Z=-980.000000),JointName="root")),Riding_Moving=(GravityOverride=(GravityOverride=(X=0.000000,Y=-450.000000,Z=-450.000000),JointName="root")),Riding_Moving_Targeting=(GravityOverride=(GravityOverride=(X=0.000000,Y=-200.000000,Z=-980.000000),JointName="root")),Tactical_Sprint=(GravityOverride=(GravityOverride=(X=0.000000,Y=-450.000000,Z=-500.000000),JointName="root")),Rail_Grinding=(GravityOverride=(GravityOverride=(X=0.000000,Y=0.000000,Z=-980.000000),JointName="root")),)
 }
 Is Wearing Cape {type=boolean}
 isWearingBackbling {type=boolean}
 RBAN Params {
  type=FortRigidBodyAnimNodeParameterSet
  default=(Skydive_GoingUp=(RigidBodyAnimNode=(ComponentLinearAccScale=(X=0.000000,Y=-0.500000,Z=0.000000),ComponentLinearVelScale=(X=0.000000,Y=-0.500000,Z=0.000000),SimSpaceSettings=(WorldAlpha=1.000000,VelocityScaleZ=0.000000,DampingAlpha=0.000000,MaxLinearVelocity=300.000000,MaxAngularVelocity=4.000000,MaxLinearAcceleration=1200.000000,MaxAngularAcceleration=16.000000,ExternalLinearDragV=(X=0.000000,Y=0.000000,Z=-1.000000)))),Skydive_FloatingDown=(RigidBodyAnimNode=(ComponentLinearAccScale=(X=0.000000,Y=0.500000,Z=0.000000),ComponentLinearVelScale=(X=0.000000,Y=0.500000,Z=0.000000),SimSpaceSettings=(WorldAlpha=1.000000,VelocityScaleZ=0.500000,DampingAlpha=0.000000,MaxLinearVelocity=700.000000,MaxAngularVelocity=8.000000,MaxLinearAcceleration=2800.000000,MaxAngularAcceleration=32.000000,ExternalLinearDragV=(X=0.500000,Y=0.500000,Z=0.500000)))),Skydive_DivingDown=(RigidBodyAnimNode=(ComponentLinearAccScale=(X=0.100000,Y=0.000000,Z=0.000000),ComponentLinearVelScale=(X=0.100000,Y=0.000000,Z=0.000000),SimSpaceSettings=(WorldAlpha=1.000000,VelocityScaleZ=0.500000,DampingAlpha=0.000000,MaxLinearVelocity=700.000000,MaxAngularVelocity=8.000000,MaxLinearAcceleration=2800.000000,MaxAngularAcceleration=32.000000,ExternalLinearDragV=(X=0.500000,Y=0.500000,Z=0.500000)))),Skydive_Parachuting=(RigidBodyAnimNode=(ComponentLinearAccScale=(X=0.000000,Y=0.100000,Z=0.300000),ComponentLinearVelScale=(X=0.000000,Y=0.100000,Z=0.300000),SimSpaceSettings=(WorldAlpha=1.000000,VelocityScaleZ=0.500000,DampingAlpha=0.000000,MaxLinearVelocity=700.000000,MaxAngularVelocity=8.000000,MaxLinearAcceleration=2800.000000,MaxAngularAcceleration=32.000000,ExternalLinearDragV=(X=0.500000,Y=0.500000,Z=0.500000)))),OnGround_Standing=(RigidBodyAnimNode=(ComponentLinearAccScale=(X=0.300000,Y=0.000000,Z=0.000000),ComponentLinearVelScale=(X=0.300000,Y=0.000000,Z=0.000000),SimSpaceSettings=(WorldAlpha=1.000000,VelocityScaleZ=0.500000,DampingAlpha=0.000000,MaxLinearVelocity=500.000000,MaxAngularVelocity=6.000000,MaxLinearAcceleration=2000.000000,MaxAngularAcceleration=24.000000,ExternalLinearDragV=(X=0.200000,Y=0.200000,Z=0.200000)))),OnGround_Moving=(RigidBodyAnimNode=(ComponentLinearAccScale=(X=0.300000,Y=0.000000,Z=0.000000),ComponentLinearVelScale=(X=0.300000,Y=0.000000,Z=0.000000),SimSpaceSettings=(WorldAlpha=1.000000,VelocityScaleZ=0.500000,DampingAlpha=0.000000,MaxLinearVelocity=500.000000,MaxAngularVelocity=6.000000,MaxLinearAcceleration=2000.000000,MaxAngularAcceleration=24.000000,ExternalLinearDragV=(X=0.200000,Y=0.200000,Z=0.200000)))),OnGround_Moving_HighRotationRate=(RigidBodyAnimNode=(SimSpaceSettings=(WorldAlpha=1.000000,VelocityScaleZ=0.500000,DampingAlpha=0.000000,MaxLinearVelocity=500.000000,MaxAngularVelocity=6.000000,MaxLinearAcceleration=2000.000000,MaxAngularAcceleration=24.000000,ExternalLinearDragV=(X=0.200000,Y=0.200000,Z=0.200000)))),OnGround_Crouch_Moving=(RigidBodyAnimNode=(SimSpaceSettings=(WorldAlpha=1.000000,VelocityScaleZ=0.500000,DampingAlpha=0.000000,MaxLinearVelocity=400.000000,MaxAngularVelocity=8.000000,MaxLinearAcceleration=3200.000000,MaxAngularAcceleration=32.000000,ExternalLinearDragV=(X=0.500000,Y=0.500000,Z=0.500000)))),Emote_or_Melee=(RigidBodyAnimNode=(SimSpaceSettings=(WorldAlpha=1.000000,VelocityScaleZ=0.500000,DampingAlpha=0.000000,MaxLinearVelocity=700.000000,MaxAngularVelocity=8.000000,MaxLinearAcceleration=2800.000000,MaxAngularAcceleration=32.000000,ExternalLinearDragV=(X=0.500000,Y=0.500000,Z=0.500000)))),Swimming=(RigidBodyAnimNode=(SimSpaceSettings=(WorldAlpha=1.000000,VelocityScaleZ=0.500000,DampingAlpha=0.000000,MaxLinearVelocity=700.000000,MaxAngularVelocity=8.000000,MaxLinearAcceleration=2800.000000,MaxAngularAcceleration=32.000000,ExternalLinearDragV=(X=0.500000,Y=0.500000,Z=0.500000)))),Front_End=(RigidBodyAnimNode=(SimSpaceSettings=(WorldAlpha=0.600000,VelocityScaleZ=0.500000,MaxLinearVelocity=100.000000,MaxAngularVelocity=1.000000,MaxLinearAcceleration=500.000000,MaxAngularAcceleration=1.000000))),DBNO=(RigidBodyAnimNode=(SimSpaceSettings=(WorldAlpha=1.000000,VelocityScaleZ=0.500000,DampingAlpha=0.000000,MaxLinearVelocity=700.000000,MaxAngularVelocity=8.000000,MaxLinearAcceleration=2800.000000,MaxAngularAcceleration=32.000000,ExternalLinearDragV=(X=0.500000,Y=0.500000,Z=0.500000)))),DBNO_Carried=(RigidBodyAnimNode=(SimSpaceSettings=(WorldAlpha=1.000000,VelocityScaleZ=0.500000,DampingAlpha=0.000000,MaxLinearVelocity=700.000000,MaxAngularVelocity=8.000000,MaxLinearAcceleration=2800.000000,MaxAngularAcceleration=32.000000,ExternalLinearDragV=(X=0.500000,Y=0.500000,Z=0.500000)))),OnGround_Falling=(RigidBodyAnimNode=(SimSpaceSettings=(WorldAlpha=1.000000,VelocityScaleZ=0.300000,DampingAlpha=0.000000,MaxLinearVelocity=700.000000,MaxAngularVelocity=8.000000,MaxLinearAcceleration=2800.000000,MaxAngularAcceleration=32.000000))),Ziplining=(RigidBodyAnimNode=(SimSpaceSettings=(WorldAlpha=1.000000,VelocityScaleZ=0.500000,DampingAlpha=0.000000,MaxLinearVelocity=700.000000,MaxAngularVelocity=8.000000,MaxLinearAcceleration=2800.000000,MaxAngularAcceleration=32.000000,ExternalLinearDragV=(X=0.500000,Y=0.500000,Z=0.500000)))),Skydive_Surfing=(RigidBodyAnimNode=(SimSpaceSettings=(WorldAlpha=1.000000,VelocityScaleZ=0.500000,DampingAlpha=0.000000,MaxLinearVelocity=700.000000,MaxAngularVelocity=8.000000,MaxLinearAcceleration=2800.000000,MaxAngularAcceleration=32.000000,ExternalLinearDragV=(X=0.500000,Y=0.500000,Z=0.500000)))),Riding_Standing=(RigidBodyAnimNode=(SimSpaceSettings=(WorldAlpha=1.000000,VelocityScaleZ=0.500000,DampingAlpha=0.000000,MaxLinearVelocity=700.000000,MaxAngularVelocity=8.000000,MaxLinearAcceleration=2800.000000,MaxAngularAcceleration=32.000000,ExternalLinearDragV=(X=0.500000,Y=0.500000,Z=0.500000)))),Riding_Moving=(RigidBodyAnimNode=(SimSpaceSettings=(WorldAlpha=1.000000,VelocityScaleZ=0.500000,DampingAlpha=0.000000,MaxLinearVelocity=700.000000,MaxAngularVelocity=8.000000,MaxLinearAcceleration=2800.000000,MaxAngularAcceleration=32.000000,ExternalLinearDragV=(X=0.500000,Y=0.500000,Z=0.500000)))),Riding_Moving_Targeting=(RigidBodyAnimNode=(SimSpaceSettings=(WorldAlpha=1.000000,VelocityScaleZ=0.500000,DampingAlpha=0.000000,MaxLinearVelocity=700.000000,MaxAngularVelocity=8.000000,MaxLinearAcceleration=2800.000000,MaxAngularAcceleration=32.000000,ExternalLinearDragV=(X=0.500000,Y=0.500000,Z=0.500000)))),Tactical_Sprint=(RigidBodyAnimNode=(ComponentLinearAccScale=(X=0.300000,Y=0.000000,Z=0.000000),ComponentLinearVelScale=(X=0.300000,Y=0.000000,Z=0.000000),SimSpaceSettings=(WorldAlpha=1.000000,VelocityScaleZ=0.250000,DampingAlpha=0.000000,MaxLinearVelocity=500.000000,MaxAngularVelocity=6.000000,MaxLinearAcceleration=2000.000000,MaxAngularAcceleration=24.000000,ExternalLinearDragV=(X=0.200000,Y=0.200000,Z=0.200000)))),Rail_Grinding=(RigidBodyAnimNode=(SimSpaceSettings=(WorldAlpha=1.000000,VelocityScaleZ=0.500000,DampingAlpha=0.000000,MaxLinearVelocity=700.000000,MaxAngularVelocity=8.000000,MaxLinearAcceleration=2800.000000,MaxAngularAcceleration=32.000000,ExternalLinearDragV=(X=0.500000,Y=0.500000,Z=0.500000)))),)
 }
 RBAN Settings Out {
  type=FortRigidBodyAnimNodeParameters
  default {simSpaceSettings {
    dampingAlpha=0.5
    maxAngularAcceleration=40
    maxAngularVelocity=10
    maxLinearAcceleration=2400
    maxLinearVelocity=600
    velocityScaleZ=0
    worldAlpha=1
   }}
 }
]
